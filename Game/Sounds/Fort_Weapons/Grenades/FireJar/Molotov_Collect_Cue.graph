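# Sound cue for collecting a Molotov (FireJar grenade), judging by its name and waves.
# A random node picks one of four equally weighted collect wave variations.
# It plays in the gadget-items sound class at reduced volume, using the shared gadget-placement attenuation.
# A concurrency override caps simultaneous instances at a small count.
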SoundCue /Game/Sounds/Fort_Weapons/Grenades/FireJar/Molotov_Collect_Cue
root=Random_0
nodes [
 {id=Random_0, children [WavePlayer_1,WavePlayer_2,WavePlayer_3,WavePlayer_4], properties {weights [1,1,1,1]}}
 {id=WavePlayer_1, properties {soundWaveAssetPtr=&/Molotov_Collect_01}}
 {id=WavePlayer_2, properties {soundWaveAssetPtr=&/Molotov_Collect_02}}
 {id=WavePlayer_3, properties {soundWaveAssetPtr=&/Molotov_Collect_03}}
 {id=WavePlayer_4, properties {soundWaveAssetPtr=&/Molotov_Collect_04}}
]
properties {
 attenuationSettings=/Game/Sounds/Attenuation/Fort_GadgetPlace_Att
 bOverrideConcurrency=true
 maxDistance=2000
 soundClassObject=&Fort_Gadgets_Items
 volumeMultiplier=0.6
 concurrencyOverrides {maxCount=2}
}
