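# Sound cue for the looping ambient "cook" sound of the FireJar Molotov grenade.
# A single looping wave player plays the Molotov_CookAmbientLoop wave at a boosted volume as the cue's root.
# It uses the gadget-placement attenuation and the gadget items sound class, and overrides concurrency to cap simultaneous instances at two.
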SoundCue /Game/Sounds/Fort_Weapons/Grenades/FireJar/Molotov_CookAmbientLoop_Cue
root=WavePlayer_1
nodes [{id=WavePlayer_1, properties {bLooping=true, soundWaveAssetPtr=&/Molotov_CookAmbientLoop}}]
properties {
 attenuationSettings=/Game/Sounds/Attenuation/Fort_GadgetPlace_Att
 bOverrideConcurrency=true
 maxDistance=2000
 soundClassObject=&Fort_Gadgets_Items
 volumeMultiplier=1.5
 concurrencyOverrides {maxCount=2}
}
