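# Sound cue for the cooking sound of the FireJar (Molotov) grenade.
# A Random node picks between two equally weighted cook wave variants, excluded from branch culling, and feeds a Modulator for variation.
# It plays on the gadgets and items sound class with weapon reload attenuation, and a concurrency override limits simultaneous instances.
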SoundCue /Game/Sounds/Fort_Weapons/Grenades/FireJar/Molotov_Cook_Cue
root=Modulator_0
nodes [
 {id=Modulator_0, children [Random_1]}
 {id=Random_1, children [WavePlayer_0,WavePlayer_5], properties {bShouldExcludeFromBranchCulling=true, weights [1,1]}}
 {id=WavePlayer_0, properties {soundWaveAssetPtr=&/Molotov_Cook_01}}
 {id=WavePlayer_5, properties {soundWaveAssetPtr=&/Molotov_Cook_02}}
]
properties {
 attenuationSettings=&Fort_Weapon_Reload_Att
 bOverrideConcurrency=true
 maxDistance=2250
 soundClassObject=&Fort_Gadgets_Items
 volumeMultiplier=0.9
 concurrencyOverrides {maxCount=2}
}
