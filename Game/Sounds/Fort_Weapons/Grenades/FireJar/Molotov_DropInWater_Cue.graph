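# Sound cue for a Molotov (FireJar grenade) dropping into water, using weapon-drop attenuation and the loot-collect sound class.
# It mixes a slightly delayed random DropInWater splash with a quieter, pitched-up random FireflyExplosion layer, each drawn from three equally weighted waves.
# An outer modulator at default settings varies the mixed result.
# A concurrency override allows one instance per owner and prevents new plays while it is active.
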SoundCue /Game/Sounds/Fort_Weapons/Grenades/FireJar/Molotov_DropInWater_Cue
root=Modulator_0
nodes [
 {id=Delay_0, children [Random_0], properties {delayMax=0.1, delayMin=0.1}}
 {id=Mixer_0, children [Delay_0,Modulator_1], properties {inputVolume [1,0.5]}}
 {id=Modulator_0, children [Mixer_0]}
 {id=Modulator_1, children [Random_1], properties {pitchMax=1.6, pitchMin=1.5}}
 {id=Random_0, children [WavePlayer_1,WavePlayer_2,WavePlayer_3], properties {weights [1,1,1]}}
 {id=Random_1, children [WavePlayer_0,WavePlayer_4,WavePlayer_5], properties {weights [1,1,1]}}
 {id=WavePlayer_0, properties {soundWaveAssetPtr=&/Molotov_FireflyExplosion_01}}
 {id=WavePlayer_1, properties {soundWaveAssetPtr=&/Molotov_DropInWater_01}}
 {id=WavePlayer_2, properties {soundWaveAssetPtr=&/Molotov_DropInWater_02}}
 {id=WavePlayer_3, properties {soundWaveAssetPtr=&/Molotov_DropInWater_03}}
 {id=WavePlayer_4, properties {soundWaveAssetPtr=&/Molotov_FireflyExplosion_02}}
 {id=WavePlayer_5, properties {soundWaveAssetPtr=&/Molotov_FireflyExplosion_03}}
]
properties {
 attenuationSettings=&Fort_Weapon_Drop_Att
 bOverrideConcurrency=true
 maxDistance=3000
 soundClassObject=&Fort_Loot_Collect
 volumeMultiplier=1.7231
 concurrencyOverrides {bLimitToOwner=true, maxCount=1, resolutionRule=PreventNew}
}
