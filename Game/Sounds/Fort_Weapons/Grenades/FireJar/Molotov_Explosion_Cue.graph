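# Sound cue for the fire jar (Molotov) grenade's explosion.
# A random node picks one of two explosion waves with equal weight.
# It uses a medium explosion attenuation and the large explosives sound class, keeps playing when silent, and references a Molotov concurrency set alongside override settings that stop the oldest instance.
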SoundCue /Game/Sounds/Fort_Weapons/Grenades/FireJar/Molotov_Explosion_Cue
root=Random_3
nodes [
 {id=Random_3, children [WavePlayer_0,WavePlayer_1], properties {weights [1,1]}}
 {id=WavePlayer_0, properties {soundWaveAssetPtr=&/Molotov_Explosion_01}}
 {id=WavePlayer_1, properties {soundWaveAssetPtr=&/Molotov_Explosion_02}}
]
properties {
 attenuationSettings=/Game/Sounds/Attenuation/Fort_Explosion_Medium_StW_Att
 maxDistance=7000
 soundClassObject=/Game/Sounds/FortSoundClassesAndModes/Fort_Explosives_Large_PSM
 virtualizationMode=PlayWhenSilent
 volumeMultiplier=1.0614
 concurrencyOverrides {maxCount=3, resolutionRule=StopOldest}
 concurrencySet [&/Molotov_Conc]
}
