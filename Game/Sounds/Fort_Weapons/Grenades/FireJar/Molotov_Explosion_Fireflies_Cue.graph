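# Sound cue for the explosion of the FireJar grenade's firefly molotov variant.
# A random node picks one of five equally weighted firefly explosion waves per play.
# It uses medium explosion attenuation and the large explosives sound class, and keeps playing when silent.
# A molotov concurrency set is assigned, alongside override settings that stop the oldest instance beyond a small cap.
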SoundCue /Game/Sounds/Fort_Weapons/Grenades/FireJar/Molotov_Explosion_Fireflies_Cue
root=Random_0
nodes [
 {id=Random_0, children [WavePlayer_2,WavePlayer_3,WavePlayer_4,WavePlayer_5,WavePlayer_6], properties {weights [1,1,1,1,1]}}
 {id=WavePlayer_2, properties {soundWaveAssetPtr=&/Molotov_FireflyExplosion_05}}
 {id=WavePlayer_3, properties {soundWaveAssetPtr=&/Molotov_FireflyExplosion_01}}
 {id=WavePlayer_4, properties {soundWaveAssetPtr=&/Molotov_FireflyExplosion_02}}
 {id=WavePlayer_5, properties {soundWaveAssetPtr=&/Molotov_FireflyExplosion_03}}
 {id=WavePlayer_6, properties {soundWaveAssetPtr=&/Molotov_FireflyExplosion_04}}
]
properties {
 attenuationSettings=/Game/Sounds/Attenuation/Fort_Explosion_Medium_StW_Att
 maxDistance=7000
 soundClassObject=/Game/Sounds/FortSoundClassesAndModes/Fort_Explosives_Large_PSM
 virtualizationMode=PlayWhenSilent
 volumeMultiplier=0.8491
 concurrencyOverrides {maxCount=3, resolutionRule=StopOldest}
 concurrencySet [&/Molotov_Conc]
}
