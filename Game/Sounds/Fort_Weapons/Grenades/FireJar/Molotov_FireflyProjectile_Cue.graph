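# Sound cue for the firefly projectile sound of the FireJar Molotov grenade.
# A random node picks between two equally weighted projectile wave variations.
# It uses a close explosion attenuation asset and the third-person ranged weapon fire sound class, with a slight volume boost.
# Overridden concurrency caps simultaneous instances, and virtualization keeps it playing when silent.
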SoundCue /Game/Sounds/Fort_Weapons/Grenades/FireJar/Molotov_FireflyProjectile_Cue
root=Random_0
nodes [
 {id=Random_0, children [WavePlayer_2,WavePlayer_3], properties {weights [1,1]}}
 {id=WavePlayer_2, properties {soundWaveAssetPtr=&/Molotov_FirelflyProjectile_01}}
 {id=WavePlayer_3, properties {soundWaveAssetPtr=&/Molotov_FirelflyProjectile_02}}
]
properties {
 attenuationSettings=/Game/Sounds/Attenuation/Fort_Explosion_Close_StW_Att
 bOverrideConcurrency=true
 maxDistance=3500
 soundClassObject=&Fort_WeaponRanged_Fire_3P_PSM
 virtualizationMode=PlayWhenSilent
 volumeMultiplier=1.2
 concurrencyOverrides {maxCount=4}
}
