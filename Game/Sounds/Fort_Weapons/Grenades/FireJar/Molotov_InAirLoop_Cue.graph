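# Looping in-air sound cue for the Molotov (FireJar) grenade, routed to the Fort_Gadgets_Items sound class.
# A looping wave player for Molotov_InAirLoop feeds an attenuation node that uses the RPG explosion attenuation.
# The cue itself references a separate projectile in-air loop attenuation, differing from the node's setting.
# A concurrency override limits the cue to a single instance, and its volume is slightly reduced.
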SoundCue /Game/Sounds/Fort_Weapons/Grenades/FireJar/Molotov_InAirLoop_Cue
root=Attenuation_1
nodes [
 {id=Attenuation_1, children [WavePlayer_0], properties {attenuationSettings=/Game/Sounds/Attenuation/Fort_Explosion_RPG_Att}}
 {id=WavePlayer_0, properties {bLooping=true, soundWaveAssetPtr=&/Molotov_InAirLoop}}
]
properties {
 attenuationSettings=/Game/Sounds/Attenuation/Fort_Weapon_Projectile_InAir_Loop_Att
 bOverrideConcurrency=true
 maxDistance=13000
 soundClassObject=&Fort_Gadgets_Items
 volumeMultiplier=0.85
 concurrencyOverrides {maxCount=1}
}
