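# Sound cue for the throw sound of the FireJar (Molotov) grenade.
# A single Molotov_Throw wave player feeds a Modulator node at its usual settings.
# It uses gadget-placement attenuation and the gadgets/items sound class at reduced volume, with overridden concurrency capping simultaneous plays and retrigger time.
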SoundCue /Game/Sounds/Fort_Weapons/Grenades/FireJar/Molotov_Throw_Cue
root=Modulator_0
nodes [
 {id=Modulator_0, children [WavePlayer_0]}
 {id=WavePlayer_0, properties {soundWaveAssetPtr=&/Molotov_Throw}}
]
properties {
 attenuationSettings=/Game/Sounds/Attenuation/Fort_GadgetPlace_Att
 bOverrideConcurrency=true
 maxDistance=2000
 soundClassObject=&Fort_Gadgets_Items
 volumeMultiplier=0.7
 concurrencyOverrides {maxCount=3, retriggerTime=0.2}
}
